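# Battle bus engine idle sound cue for Athena, layering a looping flying-wind bed with intermittent horn beeps under flying-bus attenuation.
# A quality-level node selects between two mixes, each shared by two levels.
# One mix pairs the standard wind loop with a random pick of three beeps; the other uses the optimized loop and a single beep.
# In both, beeps repeat through looping, random delays and slight volume modulation.
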SoundCue /Game/Athena/Sounds/Vehicles/Bus/S01/Athena_Bus_Engine_Idle_Cue
root=QualityLevel_0
nodes [
 {id=Delay_0, children [WavePlayer_4], properties {delayMax=20, delayMin=5}}
 {id=Delay_2, children [Random_1], properties {delayMax=20, delayMin=5}}
 {id=Looping_0, children [Modulator_0]}
 {id=Looping_2, children [Modulator_2]}
 {id=Mixer_0, children [Looping_2,WavePlayer_7], properties {inputVolume [1,2]}}
 {id=Mixer_1, children [Looping_0,WavePlayer_2], properties {inputVolume [1,2]}}
 {id=Modulator_0, children [Delay_0], properties {pitchMax=1, pitchMin=1, volumeMax=1.2, volumeMin=0.8}}
 {id=Modulator_2, children [Delay_2], properties {pitchMax=1, pitchMin=1, volumeMax=1.2, volumeMin=0.8}}
 {id=QualityLevel_0, children [Mixer_0,Mixer_0,Mixer_1,Mixer_1]}
 {id=Random_1, children [WavePlayer_0,WavePlayer_1,WavePlayer_8], properties {weights [1,1,1]}}
 {id=WavePlayer_0, properties {soundWaveAssetPtr=&/SFX_FlyingBus_BeepBeep_01}}
 {id=WavePlayer_1, properties {soundWaveAssetPtr=&/SFX_FlyingBus_BeepBeep_02}}
 {id=WavePlayer_2, properties {bLooping=true, soundWaveAssetPtr=&/SFX_Bus_Flying_Wind_Loop_opt}}
 {id=WavePlayer_4, properties {soundWaveAssetPtr=&SFX_FlyingBus_BeepBeep_01}}
 {id=WavePlayer_7, properties {bLooping=true, soundWaveAssetPtr=&/SFX_Bus_Flying_Wind_Loop}}
 {id=WavePlayer_8, properties {soundWaveAssetPtr=&/SFX_FlyingBus_BeepBeep_03}}
]
properties {
 attenuationSettings=/Game/Sounds/Attenuation/Fort_Bus_Flying_Athena_Att
 maxDistance=8000
 soundClassObject=&Fort_Gadgets
 volumeMultiplier=1
}
