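# Sound cue for the Battle Bus ignition in Athena, playing a beep-beep horn followed by a delayed engine ignition on the Battle Bus sound class.
# A root mixer blends close and distant attenuation branches, with the close one weighted louder.
# Each branch mixes its own horn wave with an ignition wave delayed one second.
# Both branches use distant-named ignition waves, the distant branch taking the "opt" variant.
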SoundCue /Game/Athena/Sounds/Vehicles/Bus/S01/Athena_Bus_Ignition_Cue
root=Mixer_1
nodes [
 {id=Attenuation_3, children [Mixer_0], properties {attenuationSettings=/Game/Sounds/Attenuation/Fort_Bus_Lobby_Engine_Athena_Distant_Att}}
 {id=Attenuation_4, children [Mixer_3], properties {attenuationSettings=/Game/Sounds/Attenuation/Fort_Bus_Lobby_Engine_Athena_Close_Att}}
 {id=Delay_1, children [WavePlayer_3], properties {delayMax=1, delayMin=1}}
 {id=Delay_3, children [WavePlayer_6], properties {delayMax=1, delayMin=1}}
 {id=Mixer_0, children [WavePlayer_2,Delay_1], properties {inputVolume [1.5,0.75]}}
 {id=Mixer_1, children [Attenuation_4, Attenuation_3], properties {inputVolume [1.3,0.75]}}
 {id=Mixer_3, children [WavePlayer_5,Delay_3], properties {inputVolume [1,1]}}
 {id=WavePlayer_2, properties {soundWaveAssetPtr=&/SFX_FlyingBus_BeepBeep_03_Distant}}
 {id=WavePlayer_3, properties {soundWaveAssetPtr=&/SFX_FlyingBus_Ignition_Distant_opt}}
 {id=WavePlayer_5, properties {soundWaveAssetPtr=&/SFX_FlyingBus_BeepBeep_03}}
 {id=WavePlayer_6, properties {soundWaveAssetPtr=&/SFX_FlyingBus_Ignition_Distant}}
]
properties {maxDistance=50000, soundClassObject=/Game/Sounds/FortSoundClassesAndModes/Fort_BattleBus, volumeMultiplier=1.25}
