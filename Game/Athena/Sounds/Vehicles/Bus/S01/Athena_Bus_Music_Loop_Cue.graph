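# Sound cue for the Season 1 Battle Bus party music loop in Athena, played while the bus flies.
# A single looping wave player feeds the root with the flying party music wave.
# It uses the bus music attenuation asset and the gadgets sound class, with boosted volume.
# Its override settings have attenuation and spatialization switched off.
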SoundCue /Game/Athena/Sounds/Vehicles/Bus/S01/Athena_Bus_Music_Loop_Cue
root=WavePlayer_10
nodes [{id=WavePlayer_10, properties {bLooping=true, soundWaveAssetPtr=&/SFX_Bus_Flying_PartyMusic_Loop}}]
properties {
 attenuationSettings=&/Fort_Bus_Music_Athena_Att
 maxDistance=21000
 soundClassObject=&Fort_Gadgets
 volumeMultiplier=1.4
 attenuationOverrides {bAttenuate=false, bSpatialize=false}
}
